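# Translucent, unlit, two-sided outline glow material for WormChalk emote effects, usable on Niagara sprites and mesh particles.
# FXCommon_SoftOutline produces outline and inner glow, with optional custom stencil, quality-scaled blur steps, and custom depth forced off on forward and mobile paths.
# Its result drives depth-faded opacity, scaled by color and particle alpha, with optional view-angle fade.
# Emissive multiplies color by particle color; the stencil branch adds the stencil index times zero.
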Material /BRCosmetics/Animation/Game/MainPlayer/Emotes/WormChalk/FX/M_Outline_Glow_Stencil
graph {
 properties {
  bUsedWithNiagaraMeshParticles=true
  bUsedWithNiagaraSprites=true
  blendMode=BLEND_Translucent
  pixelDepthOffsetMode=PDOM_Legacy
  shadingModel=MSM_Unlit
  twoSided=true
  usedShadingModels=Unlit
 }
 nodes [
  n1 = ScalarParameter() {defaultValue=40, group=Custom Stencil, parameterName=GlowMeshStencilIndex}
  n2 = StaticBoolParameter() {defaultValue=true, parameterName=Use CustomDepth}
  n3 = StaticBool()
  n4 = StaticBoolParameter() {group=Custom Stencil, parameterName="UseCustomStencil (SB)"}
  n5 = Constant() {r=6}
  n6 = Constant() {r=4}
  n7 = MaterialFunctionCall!(
   GlowMeshStencilIndex=@n1
   UseCustomDepth=ShadingPathSwitch(Default=@n2, Deferred=@n2, Forward=@n3, Mobile=@n3)
   UseCustomStencil=@n4
   OutlineThickness=ScalarParameter() {defaultValue=2, parameterName="OutlineThickness (S)"}
   OutlineStrength
   PixelDepthOffset
   MaxBlurDepth=ScalarParameter() {defaultValue=200, parameterName="MaxBlurDepth (S)"}
   ------EnableInnerGlow------=ScalarParameter() {defaultValue=1, parameterName=EnableInnerGlow}
   InnerGlowMaxArea=ScalarParameter() {defaultValue=100, parameterName="InnerGlowMaxArea (S)"}
   InnerGlowPower=ScalarParameter() {defaultValue=5, parameterName="InnerGlowPower (S)"}
   InnerGlowStrength=ScalarParameter() {defaultValue=0.1, parameterName="InnerGlowStrength (S)"}
   ------LessUse------
   MeshPosition
   Distance Steps=QualitySwitch(Default=@n5, Low=Constant() {r=3}, High=@n5, Medium=Constant() {r=5}, Epic=@n5)
   Radial Steps=QualitySwitch(Default=@n6, Low=Constant() {r=2}, High=@n6, Medium=Constant() {r=3}, Epic=@n6)
  ) {materialFunction=/Game/Effects/Fort_Effects/Effects/Atlas/Materials/FXCommon_SoftOutline}
  n8 = ScalarParameter() {defaultValue=200, parameterName=DepthFade}
  n9 = VectorParameter() {parameterName=EmissiveColor, defaultValue=<rgba 10 5.2001 1.7058 1>}
  n10 = ParticleColor()
  n11 = Multiply(A=Multiply(A=DepthFade(Opacity=@n7.Result, FadeDistance=@n8), B=@n9.A), B=@n10.A)
  n12 = ScalarParameter() {defaultValue=2, parameterName=FadeBasedOnViewAngleSensitivity}
  LinearInterpolate(
   A=@n11
   B=Multiply(
    A=@n11
    B=MaterialFunctionCall("FadeSensitivity (optional)"=@n12) {materialFunction=&FadeBasedOnViewAngle}.Result
   )
   Alpha=ScalarParameter() {parameterName=UseFadeBasedOnViewAngle}
  )
  n13 = Multiply(A=@n9.RGB, B=@n10.RGB)
  StaticSwitch(True=Add(A=Multiply(A=@n1) {constB=0}, B=@n13), False=@n13, Value=@n4)
 ]
}
