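# Niagara system for the deploy effect of the WormChalk player emote, a violet charge-up and burst followed by a lingering flare.
# Inward-moving motes, flipbook puffs and airflow meshes lead into a delayed burst of flame sprites, rotating smoke-ring meshes and more airflow.
# A long-lived flare flashes, holds with random size flicker, then pops and fades.
# Front-end, platform and time-of-day functions scale brightness, opacity and flare camera offset.
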NiagaraSystem /BRCosmetics/Animation/Game/MainPlayer/Emotes/WormChalk/FX/NS_Emote_WormChalk_deploy
dependencies {
 usedDataInterfaces [
  &NiagaraDataInterfaceCurve
  &NiagaraDataInterfaceColorCurve
  &NiagaraDataInterfacePlatformSet
  &NiagaraDataInterfaceSpriteRendererInfo
 ]
 usedDynamicInputs []
 usedModules [&SystemState]
}
emitters {
 Blue001 {
  emitterUpdateScript [
   &EmitterState(Inactive Response=::NewEnumerator1, Life Cycle Mode=::NewEnumerator1, Loop Behavior=::NewEnumerator1, Loop Delay=0.25, UseLoopDelay=true)
   &SpawnBurst_Instantaneous(Spawn Count=1, Use Spawn Probability=false)
   &SpawnBurst_Instantaneous(Spawn Count=1, Spawn Time=0.1, Use Spawn Probability=false) {module=SpawnBurst_Instantaneous001}
  ]
  particleSpawnScript [
   &InitializeParticle(
    Color=<rgba 1.7146 0 8 1>
    Lifetime=0.4
    Position Offset=<0 5 10>
    Sprite Rotation Mode=::NewEnumerator2
    Sprite Size Mode=::NewEnumerator3
    Uniform Sprite Size=18
    UsePositionOffset=true
    Write Material Random=true
   ) {module=InitializeParticle001}
   &CameraOffset(Camera Offset Amount=20)
   &AddVelocity(Velocity=<0 0 5>)
  ]
  particleUpdateScript [
   &ParticleState()
   &SolveForcesAndVelocity(Clamp Velocity=false, Limit Acceleration=false)
   &ScaleColor(
    Scale Alpha=&Multiply_Float(
     A=&FloatFromCurve(FloatCurve=DataInterface() {propertyValues {Curve {keys [
          {arriveTangent=-1.3961, time=0, value=0}
          {arriveTangent=1.1111, leaveTangent=1.1111, time=0.2, value=1}
          {arriveTangent=-1.1111, leaveTangent=-1.1111, time=0.8, value=1}
          {arriveTangent=-1.1868, leaveTangent=-1.1868, time=1, value=0}
         ]}}})
     B=&Multiply_Float(
      A=1
      B=&FN_FrontEndFloatModifier(
       Gameplay=&FN_PickFloatDependOnPlatform(TargetPlatformActive=1.2, TargetPlatformNoActive=0.8)
       Front End=0.6
      )
     )
    )
    Scale RGB=&Multiply_VectorByFloat(
     Vector=<1 1 1>
     Float=&Multiply_Float(A=1, B=&FN_FrontEndFloatModifier(Gameplay=1, Front End=0.5))
    )
    ScaleA=true
    ScaleRGB=true
   )
   &ScaleSpriteSize(Scale Factor=&Vector2DFromFloat(Value=&FloatFromCurve(FloatCurve=DataInterface() {propertyValues {Curve {keys [
          {arriveTangent=3.0495, interpMode=RCIM_Cubic, leaveTangent=3.0495, tangentMode=RCTM_User, time=0, value=0}
          {arriveTangent=1.3, interpMode=RCIM_Cubic, leaveTangent=1.3, time=0.3, value=0.95}
          {arriveTangent=0.1206, interpMode=RCIM_Cubic, leaveTangent=0.1206, tangentMode=RCTM_User, time=1, value=1.3}
         ]}}})))
   &SubUVAnimation(End Frame=12, Start Frame=5)
   /SetVariables(Particles.Dark color:LinearColor=&MakeLinearColorFromVectorAndFloat(
     "Vector (RGB)"=&MakeVectorFromLinearColorRGB(Color=<rgba 0.0107 0 0.05 1>)
     "Float (Alpha)"=&Multiply_Float(
      A=&Multiply_Float(A=1.5, B=1)
      B=&Multiply_Float(
       A=&FloatFromCurve(FloatCurve=DataInterface() {propertyValues {Curve {keys [
            {arriveTangent=-0.1811, interpMode=RCIM_Cubic, leaveTangent=-0.1811, tangentMode=RCTM_User, time=0, value=1}
            {arriveTangent=-0.6471, interpMode=RCIM_Cubic, leaveTangent=-0.6471, tangentMode=RCTM_User, time=1, value=0}
           ]}}})
       B=1
      )
     )
    )) {isSetParameters=true, module=SetVariables_#1}
   &DynamicMaterialParameters(
    Index 0 Param 1=1.2
    Module.Index0Param0Write=true
    Module.Index0Param1Write=true
    Module.Index0Param2Write=true
    Module.Index0Param3Write=true
   )
  ]
  rendererClasses [&NiagaraSpriteRendererProperties]
  renderers [
   {rendererClass=&NiagaraSpriteRendererProperties}
   {rendererClass=&NiagaraSpriteRendererProperties, rendererIndex=1}
  ]
 }
 BlueFloatParticles001 {
  emitterUpdateScript [
   &EmitterState(Inactive Response=::NewEnumerator1, Life Cycle Mode=::NewEnumerator1, Loop Behavior=::NewEnumerator1, Loop Duration=0.6, UseLoopDelay=true)
   &SpawnBurst_Instantaneous(Spawn Count=2, Use Spawn Probability=false, UseLoopCountLimit=false)
   &SpawnRate(SpawnRate=25, Use Spawn Probability=false)
  ]
  particleSpawnScript [
   &InitializeParticle(
    Lifetime Max=0.2
    Lifetime Min=0.2
    Lifetime Mode=::NewEnumerator1
    Sprite Size Mode=::NewEnumerator1
    Uniform Sprite Size Max=2.5
    Uniform Sprite Size Min=1.5
    UsePositionOffset=false
    Write Material Random=true
   ) {module=InitializeParticle001}
   &ShapeLocation(MaskSpawnGroup=false, Override Seed=false, Shape Origin=<0 5 15>, Sphere Radius=30, UseOwnerScale=true)
   &AddVelocity(Module.Use Falloff Distance=false, Origin Offset=<0 5 15>, Override Seed=false, Velocity Mode=::NewEnumerator1, Velocity Speed=-100)
   &SubUVAnimation(
    End Frame Range Override=7
    Evaluation Type=::NewEnumerator0
    "Sprite Renderer "=DataInterface() {propertyValues {SpriteRenderer=&self:BlueFloatParticles001_14.NiagaraSpriteRendererProperties_1}}
    SubUV Animation Mode:ENiagara_SubUVLookupModeV2=::NewEnumerator2
    UseEndFrame=true
    UseStartFrame=true
   )
   /SetVariables() {isSetParameters=true}
  ]
  particleUpdateScript [
   &ParticleState()
   &ScaleSpriteSize(Uniform Curve Sprite Scale=DataInterface() {propertyValues {Curve {keys [
        {arriveTangent=9.1926, interpMode=RCIM_Cubic, leaveTangent=4.2493, tangentMode=RCTM_Break, time=0, value=0}
        {arriveTangent=-1.7212, interpMode=RCIM_Cubic, leaveTangent=-1.7212, tangentMode=RCTM_User, time=0.5, value=1}
        {arriveTangent=-1.7325, interpMode=RCIM_Cubic, tangentMode=RCTM_Break, time=1, value=0}
       ]}}})
   &Drag(Rotational Drag=2, Use Linear Drag=true, Use Rotational Drag=true)
   &SolveForcesAndVelocity(Clamp Velocity=false, Limit Acceleration=false)
   &ScaleSpriteSizeBySpeed(Max Scale Factor=<1 1.5>)
   &ScaleColor(
    Scale Alpha=&FloatFromCurve(FloatCurve=DataInterface() {propertyValues {Curve {keys [
         {interpMode=RCIM_Cubic, tangentMode=RCTM_User, time=0, value=1}
         {interpMode=RCIM_Cubic, tangentMode=RCTM_User, time=0.5, value=1}
         {arriveTangent=-3.7328, interpMode=RCIM_Cubic, tangentMode=RCTM_Break, time=1, value=0}
        ]}}})
    Scale RGB=&MakeVectorFromLinearColorRGB(Color=<rgba 1.0716 0 5 1>)
    ScaleA=true
    ScaleRGB=true
   ) {module=ScaleColor001}
   /SetVariables(
    Particles.Color2=&MultiplyLinearColor(Linear Color=$Particles.Color, Scale Factor=<1 1 1 0.05>)
    Particles.SpriteSize2=&Multiply_Vector2D_ByFloat(Float=6, Vector2D=$Particles.SpriteSize)
   ) {isSetParameters=true}
  ]
  rendererClasses [&NiagaraSpriteRendererProperties]
  renderers [
   {rendererClass=&NiagaraSpriteRendererProperties}
   {rendererClass=&NiagaraSpriteRendererProperties, rendererIndex=1}
  ]
 }
 Flame_Explode002 {
  emitterUpdateScript [
   &EmitterState(Inactive Response=::NewEnumerator1, Life Cycle Mode=::NewEnumerator1, Loop Behavior=::NewEnumerator1, Loop Delay=0.7, UseLoopDelay=true)
   &SpawnBurst_Instantaneous(Spawn Count=20, Use Spawn Probability=false)
  ]
  particleSpawnScript [
   &InitializeParticle(
    Color=<rgba 2.1433 0 10 1>
    Lifetime Max=0.6
    Lifetime Min=0.5
    Lifetime Mode=::NewEnumerator1
    Sprite Rotation Mode=::NewEnumerator2
    Sprite Size Mode=::NewEnumerator1
    Uniform Sprite Size Max=20
    Uniform Sprite Size Min=15
    UsePositionOffset=true
    Write Material Random=true
   ) {module=InitializeParticle001}
   &ShapeLocation(MaskSpawnGroup=false, Override Seed=false, Shape Origin=<0 0 15>, Sphere Radius=10, UseOwnerScale=true)
   &AddVelocity(
    Module.Use Falloff Distance=false
    Origin Offset=<0 0 15>
    Override Seed=false
    Velocity Mode=::NewEnumerator1
    Velocity Origin=<0 0 0>
    Velocity Speed=&RandomRangeFloat(Minimum=80, Maximum=100)
   )
   /SetVariables(Particles.VisibilityTag=&RandomRangeInt(Minimum=0, Maximum=1)) {isSetParameters=true}
   &CameraOffset(Camera Offset Amount=5)
  ]
  particleUpdateScript [
   &ParticleState()
   &Drag(Drag=3, Use Linear Drag=true, Use Rotational Drag=false)
   &SolveForcesAndVelocity(Clamp Velocity=false, Limit Acceleration=false)
   &SubUVAnimation(
    End Frame=15
    Start Frame:NiagaraInt32=&RandomRangeInt(Minimum=3, Maximum=6)
    SubUV Loop Count:NiagaraFloat=&RandomRangeFloat(Minimum=1, Maximum=1)
   )
   &ScaleColor(
    Scale Alpha=&Multiply_Float(
     A=&FloatFromCurve(FloatCurve=DataInterface() {propertyValues {Curve {keys [
          {arriveTangent=-1.3961, interpMode=RCIM_Cubic, leaveTangent=-1.3961, tangentMode=RCTM_User, time=0, value=1}
          {arriveTangent=-1.1868, interpMode=RCIM_Cubic, leaveTangent=-1.1868, tangentMode=RCTM_User, time=1, value=1}
         ]}}})
     B=&FN_PickFloatDependOnPlatform(TargetPlatformActive=1, TargetPlatformNoActive=0.5)
    )
    Scale RGB=&Multiply_VectorByFloat(
     Vector=<1 1 1>
     Float=&Multiply_Float(A=1, B=&FN_FrontEndFloatModifier(Gameplay=1, Front End=0.4))
    )
    ScaleA=true
    ScaleRGB=true
   )
   &ScaleSpriteSize(Scale Factor=&Vector2DFromFloat(Value=&FloatFromCurve(FloatCurve=DataInterface() {propertyValues {Curve {keys [
          {arriveTangent=0.3782, interpMode=RCIM_Cubic, leaveTangent=0.3782, tangentMode=RCTM_User, time=0, value=1}
          {arriveTangent=0.1206, interpMode=RCIM_Cubic, leaveTangent=0.1206, tangentMode=RCTM_User, time=1, value=1.3}
         ]}}})))
   /SetVariables(Particles.Dark color:LinearColor=&MakeLinearColorFromVectorAndFloat(
     "Vector (RGB)"=&MakeVectorFromLinearColorRGB(Color=<rgba 0 0 0 1>)
     "Float (Alpha)"=&Multiply_Float(
      A=1
      B=&Multiply_Float(
       A=&FloatFromCurve(FloatCurve=DataInterface() {propertyValues {Curve {keys [<at 0 1 RCIM_Cubic>, {arriveTangent=-1.1868, interpMode=RCIM_Cubic, leaveTangent=-1.1868, tangentMode=RCTM_User, time=1, value=0}]}}})
       B=1
      )
     )
    )) {isSetParameters=true, module=SetVariables_#1}
   /SetVariables(Particles.Color003:LinearColor=&MakeLinearColorFromVectorAndFloat("Vector (RGB)"=&MakeVectorFromLinearColorRGB(Color=<rgba 0.0321 0 0.15 1>), "Float (Alpha)"=2.5)) {isSetParameters=true}
  ]
  rendererClasses [&NiagaraSpriteRendererProperties]
  renderers [
   {rendererClass=&NiagaraSpriteRendererProperties}
   {rendererClass=&NiagaraSpriteRendererProperties, rendererIndex=1}
   {rendererClass=&NiagaraSpriteRendererProperties, rendererIndex=2}
  ]
 }
 Flare004 {
  emitterUpdateScript [
   &EmitterState(
    Inactive Response=::NewEnumerator1
    Life Cycle Mode=::NewEnumerator1
    Loop Behavior=::NewEnumerator1
    Loop Delay=0.7
    Loop Duration=15
    UseLoopDelay=true
   )
   &SpawnBurst_Instantaneous(Use Spawn Probability=false, UseLoopCountLimit=false)
  ]
  particleSpawnScript [
   &InitializeParticle(
    Color=<rgba 0.5358 0 2.5 2>
    Lifetime=9
    Position Offset=<0 5 12>
    Sprite Size Mode=::NewEnumerator3
    Uniform Sprite Size=150
    UsePositionOffset=true
    Write Material Random=true
   )
   &CameraOffset(Camera Offset Amount=&FN_FrontEndFloatModifier(Gameplay=50, Front End=20))
   /SetVariables(Particles.CameraOffset002=10) {isSetParameters=true}
  ]
  particleUpdateScript [
   &ParticleState()
   &ScaleSpriteSize(
    Uniform Curve Scale=&RandomRangeFloat(
     Minimum=&Lerp_Float(
      A=1
      B=0.7
      Alpha=&FloatFromCurve(FloatCurve=DataInterface() {propertyValues {Curve {keys [<at 0 0>,<at 0.1 1>]}}})
     )
     Maximum=1
     Evaluation Type=::NewEnumerator1
    )
    Uniform Curve Sprite Scale=DataInterface() {propertyValues {Curve {keys [
        <at 0 0>
        {arriveTangent=2.1667, time=0.01, value=1.1}
        <at 0.05 0.6>
        {arriveTangent=17.9642, interpMode=RCIM_Cubic, leaveTangent=17.9642, tangentMode=RCTM_User, time=0.9, value=0.6}
        {arriveTangent=42.1497, interpMode=RCIM_Cubic, leaveTangent=-28.4723, tangentMode=RCTM_Break, time=0.91, value=1}
        <at 0.98 0>
       ]}}}
   )
   &ScaleColor(
    Scale Alpha=&FloatFromCurve(
     FloatCurve=DataInterface() {propertyValues {Curve {keys [
         <at 0 0>
         {arriveTangent=2.1667, time=0.01, value=1.5}
         <at 0.05 0.4>
         <at 0.8998 0.4>
         {arriveTangent=-18.0501, interpMode=RCIM_Cubic, leaveTangent=-18.0501, tangentMode=RCTM_User, time=0.9098, value=0.6}
         <at 1 -0.0024>
        ]}}}
     Scale Curve=&Multiply_Float(
      A=&Multiply_Float(A=0.5, B=&FNTimeOfDayScalar(Night Scale=0.7, Morning Scale=1, Evening Scale=1))
      B=&Multiply_Float(
       A=&FN_FrontEndFloatModifier(Gameplay=1, Front End=0.5)
       B=&FN_PickFloatDependOnPlatform(TargetPlatformActive=1, TargetPlatformNoActive=0.8)
      )
     )
    )
    Scale RGB=&MakeVectorFromLinearColorRGB(Color=<rgba 1.5 1.5 1.5 1>)
    ScaleA=true
    ScaleRGB=true
   )
   /SetVariables(Particles.size002:Vector2f=&Multiply_Vector2D(A=$Particles.SpriteSize, B=<0.45 0.45>)) {isSetParameters=true}
   /SetVariables(Particles.Color002=&MakeLinearColorFromVectorAndFloat(
     "Vector (RGB)"=&MakeVectorFromLinearColorRGB(Color=<rgba 0.0107 0 0.05 1>)
     "Float (Alpha)"=&FloatFromCurve(
      FloatCurve=DataInterface() {propertyValues {Curve {keys [<at 0.9 1>,<at 1 0>]}}}
      Scale Curve=&FN_FrontEndFloatModifier(Gameplay=2, Front End=1.5)
     )
    )) {isSetParameters=true}
  ]
  rendererClasses [&NiagaraSpriteRendererProperties]
  renderers [
   {rendererClass=&NiagaraSpriteRendererProperties}
   {rendererClass=&NiagaraSpriteRendererProperties, rendererIndex=1}
  ]
 }
 Smoke_Ring {
  emitterUpdateScript [
   &EmitterState(Inactive Response=::NewEnumerator1, Life Cycle Mode=::NewEnumerator1, Loop Behavior=::NewEnumerator1, Loop Delay=0.6, UseLoopDelay=true)
   &SpawnBurst_Instantaneous(Spawn Count=5, Use Spawn Probability=false, UseLoopCountLimit=false)
   &SpawnBurst_Instantaneous(Spawn Count=2, Spawn Time=0.2, Use Spawn Probability=false, UseLoopCountLimit=false) {module=SpawnBurst_Instantaneous001}
  ]
  particleSpawnScript [
   &InitializeParticle(
    Lifetime Max=1
    Lifetime Min=0.6
    Lifetime Mode=::NewEnumerator1
    Mesh Scale=<1 1 5>
    Mesh Scale Mode=::NewEnumerator4
    Position Offset=<0 0 15>
    UsePositionOffset=true
    Write Material Random=true
   ) {module=InitializeParticle001}
   &InitialMeshOrientation(
    Mesh Orientation Mode=::NewEnumerator2
    Rotation=&RandomRangeVector(Minimum=<0.25 -0.1 -0.1>, Maximum=<0.25 0.1 0.1>)
    Use Rotation Vector=true
   )
   &InitialMeshOrientation(
    Mesh Orientation Mode=::NewEnumerator2
    Rotation=&RandomRangeVector(Minimum=<0 0 0>, Maximum=<0 1 0>)
    Use Rotation Vector=true
   ) {module=InitialMeshOrientation001}
  ]
  particleUpdateScript [
   &ParticleState()
   &UpdateMeshOrientation(Rotation Rate=&RandomRangeFloat(Minimum=-0.8, Maximum=-1), Rotation Vector=<0 0 1>)
   &ScaleMeshSize(Scale Factor=&Multiply_VectorByFloat(
     Vector=&Multiply_VectorByFloat(
      Vector=&MakeVector(
       X=1
       Y=1
       Z=&Multiply_Float(
        A=&FloatFromCurve(
         FloatCurve=DataInterface() {propertyValues {Curve {keys [{arriveTangent=-1.572, time=0, value=1}, {interpMode=RCIM_Cubic, tangentMode=RCTM_User, time=1, value=0.5}]}}}
         Scale Curve=0.7
        )
        B=&RandomRangeFloat(Minimum=1, Maximum=1)
       )
      )
      Float=0.8
     )
     Float=&FloatFromCurve(
      FloatCurve=DataInterface() {propertyValues {Curve {keys [
          {arriveTangent=2.2632, interpMode=RCIM_Cubic, leaveTangent=2.2632, tangentMode=RCTM_User, time=0, value=0.25}
          {arriveTangent=2.9967, interpMode=RCIM_Cubic, leaveTangent=2.9967, tangentMode=RCTM_User, time=1, value=3}
         ]}}}
      Scale Curve=&Multiply_Float(A=&RandomRangeFloat(Minimum=2, Maximum=3), B=1)
     )
    ))
   &ScaleColor(
    Scale Alpha=&FloatFromCurve(
     FloatCurve=DataInterface() {propertyValues {Curve {keys [
         {arriveTangent=6.8924, interpMode=RCIM_Cubic, leaveTangent=6.8924, tangentMode=RCTM_User, time=0.02, value=0}
         {arriveTangent=1.141, interpMode=RCIM_Cubic, leaveTangent=1.141, tangentMode=RCTM_User, time=0.1, value=0.6}
         {arriveTangent=-1.0842, interpMode=RCIM_Cubic, leaveTangent=-1.0842, tangentMode=RCTM_User, time=0.7, value=0.55}
         {arriveTangent=-1.8777, interpMode=RCIM_Cubic, leaveTangent=-1.8777, tangentMode=RCTM_User, time=1, value=0}
        ]}}}
     Scale Curve=&Multiply_Float(
      A=&Multiply_Float(A=2, B=&FNTimeOfDayScalar(Night Scale=0.5, Morning Scale=1, Evening Scale=1))
      B=1
     )
    )
    Scale RGB=&MakeVectorFromLinearColorRGB(Color=&ColorFromCurve(ColorCurve=DataInterface() {propertyValues {
        AlphaCurve {keys []}
        BlueCurve {keys [<at 0.2705 5>,<at 0.4921 0.3906>,<at 0.8581 0.02 RCIM_Cubic>]}
        GreenCurve {keys [<at 0.2705 0>,<at 0.4921 0>,<at 0.8581 0 RCIM_Cubic>]}
        RedCurve {keys [<at 0.2705 1.0716>,<at 0.4921 0.0837>,<at 0.8581 0.0043 RCIM_Cubic>]}
       }}))
    ScaleA=true
    ScaleRGB=true
   )
  ]
  rendererClasses [&NiagaraMeshRendererProperties]
  renderers [{rendererClass=&NiagaraMeshRendererProperties, rendererIndex=0}]
 }
 airflow {
  emitterUpdateScript [
   &EmitterState(
    Inactive Response=::NewEnumerator1
    Life Cycle Mode=::NewEnumerator1
    Loop Behavior=::NewEnumerator1
    Loop Delay=0.08
    Loop Duration=3
    UseLoopDelay=true
   )
   &SpawnBurst_Instantaneous(Use Spawn Probability=false, UseLoopCountLimit=false) {enabled=false}
   &SpawnRate(
    SpawnRate=&FloatFromCurve(FloatCurve=DataInterface() {propertyValues {Curve {keys [<at 0 1>,<at 1 0>]}}}, CurveIndex=$Emitter.Age, Scale Curve=20)
    Use Spawn Probability=false
   )
  ]
  particleSpawnScript [
   &InitializeParticle(
    Color=&MakeCustomLinearColorFromBool(Bool=&RandomBool(), True Color=<rgba 6.4298 0 30 1>, False Color=<rgba 2.1433 0 10 1>)
    Lifetime Max=0.3
    Lifetime Min=0.2
    Lifetime Mode=::NewEnumerator1
    Mesh Scale Max=<1.2 1.2 1.2>
    Mesh Scale Mode=::NewEnumerator2
    Position Offset=<0 0 10>
    UsePositionOffset=true
    Write Material Random=true
   )
   &InitialMeshOrientation(
    Mesh Orientation Mode=::NewEnumerator2
    Orientation Coordinate Space=::Local
    Rotation=&RandomRangeVector(Minimum=<-1 -1 -1>, Maximum=<1 1 1>)
    Rotation Coordinate Space=::NewEnumerator0
    Use Rotation Vector=true
   )
  ]
  particleUpdateScript [
   &ParticleState()
   &DynamicMaterialParameters(
    Index 0 Param 1=&FloatFromCurve(FloatCurve=DataInterface() {propertyValues {Curve {keys [{interpMode=RCIM_Cubic, tangentMode=RCTM_User, time=0, value=1}, {interpMode=RCIM_Cubic, tangentMode=RCTM_User, time=1, value=1}]}}})
    Index 0 Param 4=&Add_Float(
     A=&Multiply_Float(A=$Emitter.Age, B=5)
     B=&RandomRangeFloat(Minimum=0, Maximum=1000)
    )
    Module.Param0WriteEnabled=true
    Module.Param1WriteEnabled=true
    Module.Param2WriteEnabled=true
    Module.Param3WriteEnabled=true
   )
   &ScaleMeshSize(Scale Factor=&Multiply_Vector(
     A=<1 1 1>
     B=&VectorFromFloat(Value=&FloatFromCurve(
       FloatCurve=DataInterface() {propertyValues {Curve {keys [
           {arriveTangent=-0.1194, interpMode=RCIM_Cubic, leaveTangent=-0.1194, tangentMode=RCTM_User, time=0, value=1}
           {arriveTangent=-0.5033, interpMode=RCIM_Cubic, leaveTangent=-0.5033, tangentMode=RCTM_User, time=1, value=0.5}
          ]}}}
       Scale Curve=0.25
      ))
    ))
   &UpdateMeshOrientation(Rotation Vector=<0 0 1>)
   &ScaleColor(
    Scale Alpha=&FloatFromCurve(FloatCurve=DataInterface() {propertyValues {Curve {keys [
         {arriveTangent=4.1785, interpMode=RCIM_Cubic, leaveTangent=4.1785, tangentMode=RCTM_User, time=0, value=0}
         {arriveTangent=0.9995, interpMode=RCIM_Cubic, leaveTangent=0.9995, tangentMode=RCTM_User, time=0.3, value=0.625}
         {arriveTangent=-0.0405, interpMode=RCIM_Cubic, leaveTangent=-0.0405, tangentMode=RCTM_User, time=1, value=1}
        ]}}})
    ScaleA=true
    ScaleRGB=true
   )
  ]
  rendererClasses [&NiagaraMeshRendererProperties]
  renderers [{rendererClass=&NiagaraMeshRendererProperties, rendererIndex=0}]
 }
 airflow002 {
  emitterUpdateScript [
   &EmitterState(Inactive Response=::NewEnumerator1, Life Cycle Mode=::NewEnumerator1, Loop Behavior=::NewEnumerator1, Loop Delay=0.6, UseLoopDelay=true)
   &SpawnBurst_Instantaneous(Spawn Count=4, Use Spawn Probability=false, UseLoopCountLimit=false)
  ]
  particleSpawnScript [
   &InitializeParticle(
    Color=&MakeCustomLinearColorFromBool(Bool=&RandomBool(), True Color=<rgba 0.2143 0 1 1>, False Color=<rgba 0.0107 0 0.05 1>)
    Lifetime Max=0.8
    Lifetime Min=0.4
    Lifetime Mode=::NewEnumerator1
    Mesh Scale Max=<1.2 1.2 1.2>
    Mesh Scale Mode=::NewEnumerator2
    Position Offset=<0 0 15>
    UsePositionOffset=true
    Write Material Random=true
   )
   &InitialMeshOrientation(
    Mesh Orientation Mode=::NewEnumerator2
    Orientation Coordinate Space=::Local
    Rotation=&RandomRangeVector(Minimum=<-1 -1 -1>, Maximum=<1 1 1>)
    Rotation Coordinate Space=::NewEnumerator0
    Use Rotation Vector=true
   )
  ]
  particleUpdateScript [
   &ParticleState()
   &DynamicMaterialParameters(
    Index 0 Param 1=&FloatFromCurve(FloatCurve=DataInterface() {propertyValues {Curve {keys [{interpMode=RCIM_Cubic, tangentMode=RCTM_User, time=0, value=1}, {interpMode=RCIM_Cubic, tangentMode=RCTM_User, time=1, value=1}]}}})
    Index 0 Param 4=&Add_Float(
     A=&Multiply_Float(A=$Emitter.Age, B=5)
     B=&RandomRangeFloat(Minimum=0, Maximum=1000)
    )
    Module.Param0WriteEnabled=true
    Module.Param1WriteEnabled=true
    Module.Param2WriteEnabled=true
    Module.Param3WriteEnabled=true
   )
   &ScaleMeshSize(Scale Factor=&Multiply_Vector(
     A=<1 1 1>
     B=&VectorFromFloat(Value=&FloatFromCurve(
       FloatCurve=DataInterface() {propertyValues {Curve {keys [
           {arriveTangent=0.1666, interpMode=RCIM_Cubic, leaveTangent=0.1666, tangentMode=RCTM_User, time=0, value=0}
           {arriveTangent=0.6002, interpMode=RCIM_Cubic, leaveTangent=0.6002, tangentMode=RCTM_User, time=1, value=1}
          ]}}}
       Scale Curve=0.7
      ))
    ))
   &UpdateMeshOrientation(Rotation Vector=<0 0 1>)
   &ScaleColor(
    Scale Alpha=&FloatFromCurve(FloatCurve=DataInterface() {propertyValues {Curve {keys [
         {arriveTangent=4.1785, interpMode=RCIM_Cubic, leaveTangent=4.1785, tangentMode=RCTM_User, time=0, value=0}
         {arriveTangent=0.9995, interpMode=RCIM_Cubic, leaveTangent=0.9995, tangentMode=RCTM_User, time=0.3, value=0.625}
         {arriveTangent=-0.0405, interpMode=RCIM_Cubic, leaveTangent=-0.0405, tangentMode=RCTM_User, time=1, value=1}
        ]}}})
    ScaleA=true
    ScaleRGB=true
   )
  ]
  rendererClasses [&NiagaraMeshRendererProperties]
  renderers [{rendererClass=&NiagaraMeshRendererProperties, rendererIndex=0}]
 }
}
summary {
 systemName=NS_Emote_WormChalk_deploy
 emitters [
  {emitterName=airflow002, rendererClasses [&NiagaraMeshRendererProperties]}
  {emitterName=Smoke_Ring, rendererClasses [&NiagaraMeshRendererProperties]}
  {emitterName=Flare004, rendererClasses [&NiagaraSpriteRendererProperties]}
  {emitterName=Flame_Explode002, rendererClasses [&NiagaraSpriteRendererProperties]}
  {emitterName=BlueFloatParticles001, rendererClasses [&NiagaraSpriteRendererProperties]}
  {emitterName=airflow, rendererClasses [&NiagaraMeshRendererProperties]}
  {emitterName=Blue001, rendererClasses [&NiagaraSpriteRendererProperties]}
 ]
 userVariables [
  User.bIsFrontEnd {type=NiagaraBool}
  User.bIsFrontEndPreview {type=NiagaraBool}
 ]
}
