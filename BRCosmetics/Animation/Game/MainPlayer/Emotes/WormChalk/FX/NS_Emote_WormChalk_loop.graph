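# Looping-stage Niagara effect for the WormChalk player emote, producing purple and near-black flames, smoke rings and an edge glow.
# Black, Blue and BlackFlame002 sprites spawn on loop-age curves alongside a long-lived EdgeGlow mesh and stepped pulses of spinning, expanding Smoke_Ring002 meshes.
# Flame_Explode adds a delayed sprite burst after the main roughly eight-second window.
# Selected emitters adjust brightness, opacity, scale or camera offset by time of day, front-end context or platform.
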NiagaraSystem /BRCosmetics/Animation/Game/MainPlayer/Emotes/WormChalk/FX/NS_Emote_WormChalk_loop
dependencies {
 usedDataInterfaces [
  &NiagaraDataInterfaceCurve
  &NiagaraDataInterfaceColorCurve
  &NiagaraDataInterfacePlatformSet
  &NiagaraDataInterfaceVector2DCurve
 ]
 usedDynamicInputs [&ReturnNormalizedExecIndex]
 usedModules [&SystemState]
}
emitters {
 Black {
  emitterUpdateScript [
   &EmitterState(
    Inactive Response=::NewEnumerator1
    Life Cycle Mode=::NewEnumerator1
    Loop Behavior=::NewEnumerator1
    Loop Delay=0.8
    Loop Duration=8.3
    UseLoopDelay=true
   )
   &SpawnRate(
    SpawnRate=&FloatFromCurve(
     FloatCurve=DataInterface() {propertyValues {Curve {keys [<at 0 1>,<at 0.95 1>,<at 1 0>]}}}
     CurveIndex=$Emitter.NormalizedLoopAge
     Scale Curve=15
    )
    Use Spawn Probability=false
   )
   &SpawnBurst_Instantaneous(Spawn Count=2, Use Spawn Probability=false)
  ]
  particleSpawnScript [
   &InitializeParticle(
    Color=<rgba 0.0386 0 0.18 1>
    Lifetime Max=1
    Lifetime Min=0.6
    Lifetime Mode=::NewEnumerator1
    Sprite Rotation Mode=::NewEnumerator2
    Sprite Size Mode=::NewEnumerator1
    Uniform Sprite Size Max=20
    Uniform Sprite Size Min=18
    UsePositionOffset=true
    Write Material Random=true
   ) {module=InitializeParticle001}
   &ShapeLocation(MaskSpawnGroup=false, Override Seed=false, Shape Origin=<0 3 12>, Sphere Radius=3, UseOwnerScale=true)
   &AddVelocity(
    Module.Use Falloff Distance=false
    Origin Offset=<0 3 12>
    Override Seed=false
    Velocity Mode=::NewEnumerator1
    Velocity Origin=<0 0 0>
    Velocity Speed=40
   )
   &CameraOffset(Camera Offset Amount=16)
  ]
  particleUpdateScript [
   &ParticleState()
   &ScaleColor(
    Scale Alpha=&Multiply_Float(
     A=&FloatFromCurve(FloatCurve=DataInterface() {propertyValues {Curve {keys [<at 0.9 1 RCIM_Cubic>, {arriveTangent=-1.1868, interpMode=RCIM_Cubic, leaveTangent=-1.1868, tangentMode=RCTM_User, time=1, value=0}]}}})
     B=1
    )
    Scale RGB=&VectorFromFloat(Value=&Multiply_Float(A=1, B=&FNTimeOfDayScalar(Night Scale=0.5, Morning Scale=1, Evening Scale=1)))
    ScaleA=true
    ScaleRGB=true
   )
   &ScaleSpriteSize(Scale Factor=&Vector2DFromFloat(Value=&FloatFromCurve(
      FloatCurve=DataInterface() {propertyValues {Curve {keys [
          {arriveTangent=0.3782, interpMode=RCIM_Cubic, leaveTangent=0.3782, tangentMode=RCTM_User, time=0, value=1.1}
          {arriveTangent=0.1206, interpMode=RCIM_Cubic, leaveTangent=0.1206, tangentMode=RCTM_User, time=1, value=1.3}
         ]}}}
      Scale Curve=&RandomRangeFloat(Minimum=0.8, Maximum=1)
     )))
   /SetVariables(Particles.size002:Vector2f=&Multiply_Vector2D(A=$Particles.SpriteSize, B=&Vector2DFromFloat(Value=1.3))) {isSetParameters=true}
   /SetVariables(Particles.Color002=&Multiply_LinearColorByFloat(
     Linear Color=<rgba 0.6515 0.1354 2 1>
     Float=&FloatFromCurve(
      FloatCurve=DataInterface() {propertyValues {Curve {keys [<at 0 0>,<at 0.2 1>,<at 0.7 1>,<at 1 0>]}}}
      Scale Curve=&Multiply_Float(
       A=&Multiply_Float(A=1, B=&FNTimeOfDayScalar(Night Scale=0.5, Morning Scale=1, Evening Scale=1))
       B=&Multiply_Float(A=&FN_FrontEndFloatModifier(Gameplay=1, Front End=0.7), B=1)
      )
     )
    )) {isSetParameters=true}
   &SubUVAnimation(
    End Frame=15
    Start Frame:NiagaraInt32=&RandomRangeInt(Minimum=3, Maximum=15)
    SubUV Loop Count:NiagaraFloat=&RandomRangeFloat(Minimum=1, Maximum=1)
   )
   &Drag(Drag=3, Use Linear Drag=true, Use Rotational Drag=false)
   &SolveForcesAndVelocity(Clamp Velocity=false, Limit Acceleration=false)
  ]
  rendererClasses [&NiagaraSpriteRendererProperties]
  renderers [
   {rendererClass=&NiagaraSpriteRendererProperties}
   {rendererClass=&NiagaraSpriteRendererProperties, rendererIndex=1}
  ]
 }
 BlackFlame002 {
  emitterUpdateScript [
   &EmitterState(
    Inactive Response=::NewEnumerator1
    Life Cycle Mode=::NewEnumerator1
    Loop Behavior=::NewEnumerator1
    Loop Delay=0.6
    Loop Duration=8
    UseLoopDelay=true
   )
   &SpawnBurst_Instantaneous(Spawn Count=10, Use Spawn Probability=false, UseLoopCountLimit=false) {enabled=false}
   &SpawnRate(
    SpawnRate=&FloatFromCurve(FloatCurve=DataInterface() {propertyValues {Curve {keys [<at 0.7 1>,<at 1 0>]}}}, CurveIndex=$Emitter.NormalizedLoopAge, Scale Curve=20)
    Use Spawn Probability=false
   )
  ]
  particleSpawnScript [
   &InitializeParticle(
    Color=<rgba 0.96 0 0.7392 3.5>
    Lifetime Max=1.5
    Lifetime Mode=::NewEnumerator1
    Position Offset=<0 5 14>
    Sprite Size Max=<40 100>
    Sprite Size Min=<30 120>
    Sprite Size Mode=::NewEnumerator2
    UsePositionOffset=true
    Write Material Random=true
   )
   &SpriteFacingAndAlignment(Sprite Alignment=<0 0 1>, WriteAlignment=true, WriteFacing=false) {enabled=false}
   &CameraOffset(Camera Offset Amount=20)
   &ShapeLocation(
    Handle Radius=0
    Large Radius=2
    MaskSpawnGroup=false
    Override Seed=false
    Rotation Axis=<0 1 0>
    Rotation Mode=::NewEnumerator1
    Shape Primitive=::NewEnumerator3
    UseOwnerScale=true
   )
   &AddVelocity(Module.Use Falloff Distance=false, Origin Offset=<0 5.5 14>, Override Seed=false, Velocity Mode=::NewEnumerator1, Velocity Speed=5)
  ]
  particleUpdateScript [
   &ParticleState()
   &ScaleSpriteSize(Scale Factor=&Vector2DFromCurve(
     Vector2Curve=DataInterface() {propertyValues {
       XCurve {keys [
         {arriveTangent=1.0316, interpMode=RCIM_Cubic, leaveTangent=1.0316, tangentMode=RCTM_User, time=0, value=0.6}
         {arriveTangent=0.4, interpMode=RCIM_Cubic, leaveTangent=0.4, time=0.3, value=0.85}
         <at 1 1 RCIM_Cubic>
        ]}
       YCurve {keys [<at 0 0.75>,<at 0.3 0.9>,<at 1 1>]}
      }}
     ScaleCurve=<1 1>
    ))
   &ScaleColor(
    Scale Alpha=&FloatFromCurve(
     FloatCurve=DataInterface() {propertyValues {Curve {keys [
         <at 0 0 RCIM_Cubic>
         {arriveTangent=0.0248, interpMode=RCIM_Cubic, leaveTangent=0.0248, tangentMode=RCTM_User, time=0.1, value=1}
         {arriveTangent=-1.3658, interpMode=RCIM_Cubic, leaveTangent=-1.3658, tangentMode=RCTM_User, time=0.4, value=0.4}
         <at 1 0 RCIM_Cubic>
        ]}}}
     Scale Curve=&FN_FrontEndFloatModifier(
      Gameplay=&Multiply_Float(A=0.5, B=&FN_PickFloatDependOnPlatform(TargetPlatformActive=1, TargetPlatformNoActive=0.5))
      Front End=0.45
     )
    )
    Scale RGB=&VectorFromFloat(Value=&Multiply_Float(
      A=&FN_FrontEndFloatModifier(Gameplay=1, Front End=0.6)
      B=&FNTimeOfDayScalar(Night Scale=0.5, Morning Scale=1, Evening Scale=1)
     ))
    ScaleA=true
    ScaleRGB=true
   )
   &SolveForcesAndVelocity(Clamp Velocity=false, Limit Acceleration=false)
  ]
  rendererClasses [&NiagaraSpriteRendererProperties]
  renderers [{rendererClass=&NiagaraSpriteRendererProperties}]
 }
 Blue {
  emitterUpdateScript [
   &EmitterState(
    Inactive Response=::NewEnumerator1
    Life Cycle Mode=::NewEnumerator1
    Loop Behavior=::NewEnumerator1
    Loop Delay=0.8
    Loop Duration=8.3
    UseLoopDelay=true
   )
   &SpawnRate(
    SpawnRate=&FloatFromCurve(
     FloatCurve=DataInterface() {propertyValues {Curve {keys [<at 0 1>,<at 0.95 1>,<at 1 0>]}}}
     CurveIndex=$Emitter.NormalizedLoopAge
     Scale Curve=4
    )
    Use Spawn Probability=false
   )
   &SpawnBurst_Instantaneous(Spawn Count=2, Use Spawn Probability=false)
  ]
  particleSpawnScript [
   &InitializeParticle(
    Color=<rgba 1.0716 0 5 1>
    Lifetime Max=0.8
    Lifetime Min=0.6
    Lifetime Mode=::NewEnumerator1
    Sprite Rotation Mode=::NewEnumerator2
    Sprite Size Mode=::NewEnumerator1
    Uniform Sprite Size Max=16
    Uniform Sprite Size Min=20
    UsePositionOffset=true
    Write Material Random=true
   ) {module=InitializeParticle001}
   &ShapeLocation(MaskSpawnGroup=false, Override Seed=false, Shape Origin=<0 3 12>, Sphere Radius=4, UseOwnerScale=true)
   &AddVelocity(
    Module.Use Falloff Distance=false
    Origin Offset=<0 3 12>
    Override Seed=false
    Velocity Mode=::NewEnumerator1
    Velocity Origin=<0 0 0>
    Velocity Speed=30
   )
   &CameraOffset(Camera Offset Amount=20)
  ]
  particleUpdateScript [
   &ParticleState()
   &Drag(Drag=3, Use Linear Drag=true, Use Rotational Drag=false)
   &SolveForcesAndVelocity(Clamp Velocity=false, Limit Acceleration=false)
   &ScaleColor(
    Scale Alpha=&Multiply_Float(
     A=&FloatFromCurve(FloatCurve=DataInterface() {propertyValues {Curve {keys [
          {arriveTangent=-1.3961, time=0, value=0}
          {arriveTangent=1.1111, leaveTangent=1.1111, time=0.1, value=1}
          {arriveTangent=-1.1111, leaveTangent=-1.1111, time=0.9, value=1}
          {arriveTangent=-1.1868, leaveTangent=-1.1868, time=1, value=0}
         ]}}})
     B=&Multiply_Float(
      A=&Multiply_Float(A=1, B=&FN_FrontEndFloatModifier(Gameplay=1, Front End=0.6))
      B=1
     )
    )
    Scale RGB=&Multiply_VectorByFloat(Vector=<1 1 1>, Float=1)
    ScaleA=true
    ScaleRGB=true
   )
   &ScaleSpriteSize(Scale Factor=&Vector2DFromFloat(Value=&FloatFromCurve(FloatCurve=DataInterface() {propertyValues {Curve {keys [
          {arriveTangent=0.3782, interpMode=RCIM_Cubic, leaveTangent=0.3782, tangentMode=RCTM_User, time=0, value=1}
          {arriveTangent=0.1206, interpMode=RCIM_Cubic, leaveTangent=0.1206, tangentMode=RCTM_User, time=1, value=1.3}
         ]}}})))
   &SubUVAnimation(End Frame=15, Start Frame=2, SubUV Loop Count:NiagaraFloat=&RandomRangeFloat(Minimum=1, Maximum=1))
   /SetVariables(Particles.Dark color:LinearColor=&MakeLinearColorFromVectorAndFloat(
     "Vector (RGB)"=&MakeVectorFromLinearColorRGB(Color=<rgba 0.0107 0 0.05 1>)
     "Float (Alpha)"=&Multiply_Float(
      A=&Multiply_Float(A=1, B=1)
      B=&Multiply_Float(
       A=&FloatFromCurve(FloatCurve=DataInterface() {propertyValues {Curve {keys [
            {arriveTangent=-0.1811, interpMode=RCIM_Cubic, leaveTangent=-0.1811, tangentMode=RCTM_User, time=0, value=1}
            {arriveTangent=-0.6471, interpMode=RCIM_Cubic, leaveTangent=-0.6471, tangentMode=RCTM_User, time=1, value=0}
           ]}}})
       B=1
      )
     )
    )) {isSetParameters=true, module=SetVariables_#1}
   /SetVariables(Particles.size002:Vector2f=&Multiply_Vector2D(A=$Particles.SpriteSize, B=&Vector2DFromFloat(Value=1.2))) {isSetParameters=true}
   /SetVariables(Particles.Color002=&Multiply_LinearColorByFloat(
     Linear Color=<rgba 0.0536 0 0.25 1>
     Float=&FloatFromCurve(FloatCurve=DataInterface() {propertyValues {Curve {keys [<at 0 0>,<at 0.2 1>,<at 0.7 1>,<at 1 0>]}}})
    )) {isSetParameters=true}
   &DynamicMaterialParameters(
    Index 0 Param 1=1.2
    Module.Index0Param0Write=true
    Module.Index0Param1Write=true
    Module.Index0Param2Write=true
    Module.Index0Param3Write=true
   )
  ]
  rendererClasses [&NiagaraSpriteRendererProperties]
  renderers [
   {rendererClass=&NiagaraSpriteRendererProperties}
   {rendererClass=&NiagaraSpriteRendererProperties, rendererIndex=1}
   {rendererClass=&NiagaraSpriteRendererProperties, rendererIndex=2}
  ]
 }
 EdgeGlow {
  emitterUpdateScript [
   &EmitterState(
    Inactive Response=::NewEnumerator1
    Life Cycle Mode=::NewEnumerator1
    Loop Behavior=::NewEnumerator1
    Loop Delay=0.8
    Loop Duration=15
    UseLoopDelay=true
   )
   &SpawnBurst_Instantaneous(Use Spawn Probability=false, UseLoopCountLimit=false)
   /SetVariables(Emitter.bool:NiagaraBool=&InvertBool(Boolean:NiagaraBool=$User.bIsFrontEndPreview)) {isSetParameters=true}
  ]
  particleSpawnScript [
   &InitializeParticle(
    Color Mode=::NewEnumerator0
    Lifetime=8.2
    Mesh Scale=<1.2 0.5 1.2>
    Mesh Scale Mode=::NewEnumerator4
    Position Offset=<0 40 0>
    Sprite Size=<118.0405 198.6727>
    Sprite Size Mode=::NewEnumerator4
    UsePositionOffset=true
    Write Material Random=true
   )
   &CameraOffset(Camera Offset Amount=50)
   &InitialMeshOrientation(Mesh Orientation Mode=::NewEnumerator2, Rotation=<0 0 0.03>, Use Rotation Vector=true)
   /SetVariables(Particles.Position002:NiagaraPosition=<0 0 15>) {isSetParameters=true}
  ]
  particleUpdateScript [
   &ParticleState()
   &ScaleMeshSize(Scale Factor=&VectorFromFloat(Value=&FN_FrontEndFloatModifier(Gameplay=0.7, Front End=0.6)))
   &ScaleColor(
    Scale Alpha=&FloatFromCurve(FloatCurve=DataInterface() {propertyValues {Curve {keys [<at 0 1>, {arriveTangent=29.4027, tangentMode=RCTM_User, time=0.2, value=1.5}, <at 0.8 1.5>, <at 1 1>]}}})
    Scale RGB=&VectorFromFloat(Value=&Multiply_Float(A=1, B=&FNTimeOfDayScalar(Night Scale=1.2, Morning Scale=1, Evening Scale=1)))
    ScaleA=true
    ScaleRGB=true
   )
  ]
  rendererClasses [&NiagaraMeshRendererProperties]
  renderers [
   {rendererClass=&NiagaraMeshRendererProperties, rendererIndex=0}
   {rendererClass=&NiagaraMeshRendererProperties, rendererIndex=1}
   {rendererClass=&NiagaraMeshRendererProperties, rendererIndex=2}
  ]
 }
 Flame_Explode {
  emitterUpdateScript [
   &EmitterState(
    Inactive Response=::NewEnumerator1
    Life Cycle Mode=::NewEnumerator1
    Loop Behavior=::NewEnumerator1
    Loop Delay=9
    Loop Duration=15
    UseLoopDelay=true
   )
   &SpawnBurst_Instantaneous(Spawn Count=10, Use Spawn Probability=false)
  ]
  particleSpawnScript [
   &InitializeParticle(
    Color=<rgba 0.8573 0 4 1>
    Lifetime Max=0.5
    Lifetime Min=0.3
    Lifetime Mode=::NewEnumerator1
    Sprite Rotation Mode=::NewEnumerator2
    Sprite Size Mode=::NewEnumerator1
    Uniform Sprite Size Max=12
    Uniform Sprite Size Min=10
    UsePositionOffset=true
    Write Material Random=true
   ) {module=InitializeParticle001}
   &ShapeLocation(MaskSpawnGroup=false, Override Seed=false, Shape Origin=<0 5 15>, Sphere Radius=3, UseOwnerScale=true)
   &AddVelocity(
    Module.Use Falloff Distance=false
    Origin Offset=<0 5 15>
    Override Seed=false
    Velocity Mode=::NewEnumerator1
    Velocity Origin=<0 0 0>
    Velocity Speed=80
   )
   /SetVariables(Particles.VisibilityTag=&RandomRangeInt(Minimum=0, Maximum=1)) {isSetParameters=true}
   &CameraOffset(Camera Offset Amount=15)
  ]
  particleUpdateScript [
   &ParticleState()
   &Drag(Drag=3, Use Linear Drag=true, Use Rotational Drag=false)
   &SolveForcesAndVelocity(Clamp Velocity=false, Limit Acceleration=false)
   &SubUVAnimation(
    End Frame=15
    Start Frame:NiagaraInt32=&RandomRangeInt(Minimum=0, Maximum=5)
    SubUV Loop Count:NiagaraFloat=&RandomRangeFloat(Minimum=1, Maximum=1)
   )
   &ScaleColor(
    Scale Alpha=&Multiply_Float(
     A=&FloatFromCurve(FloatCurve=DataInterface() {propertyValues {Curve {keys [
          {arriveTangent=-1.3961, interpMode=RCIM_Cubic, leaveTangent=-1.3961, tangentMode=RCTM_User, time=0, value=1}
          {arriveTangent=-1.1868, interpMode=RCIM_Cubic, leaveTangent=-1.1868, tangentMode=RCTM_User, time=1, value=1}
         ]}}})
     B=1
    )
    Scale RGB=&Multiply_VectorByFloat(
     Vector=<1 1 1>
     Float=&Multiply_Float(A=1, B=&FN_PickFloatDependOnPlatform(TargetPlatformActive=1, TargetPlatformNoActive=0.8))
    )
    ScaleA=true
    ScaleRGB=true
   )
   &ScaleSpriteSize(Scale Factor=&Vector2DFromFloat(Value=&FloatFromCurve(FloatCurve=DataInterface() {propertyValues {Curve {keys [
          {arriveTangent=0.3782, interpMode=RCIM_Cubic, leaveTangent=0.3782, tangentMode=RCTM_User, time=0, value=1}
          {arriveTangent=0.1206, interpMode=RCIM_Cubic, leaveTangent=0.1206, tangentMode=RCTM_User, time=1, value=1.3}
         ]}}})))
   /SetVariables(Particles.Dark color:LinearColor=&MakeLinearColorFromVectorAndFloat(
     "Vector (RGB)"=&MakeVectorFromLinearColorRGB(Color=<rgba 0 0 0 1>)
     "Float (Alpha)"=&Multiply_Float(
      A=1
      B=&Multiply_Float(
       A=&FloatFromCurve(FloatCurve=DataInterface() {propertyValues {Curve {keys [<at 0 1 RCIM_Cubic>, {arriveTangent=-1.1868, interpMode=RCIM_Cubic, leaveTangent=-1.1868, tangentMode=RCTM_User, time=1, value=0}]}}})
       B=1
      )
     )
    )) {isSetParameters=true, module=SetVariables_#1}
   /SetVariables(Particles.Color003:LinearColor=&MakeLinearColorFromVectorAndFloat("Vector (RGB)"=&MakeVectorFromLinearColorRGB(Color=<rgba 0.0403 0 0.1 1>), "Float (Alpha)"=2.5)) {isSetParameters=true}
  ]
  rendererClasses [&NiagaraSpriteRendererProperties]
  renderers [
   {rendererClass=&NiagaraSpriteRendererProperties}
   {rendererClass=&NiagaraSpriteRendererProperties, rendererIndex=1}
   {rendererClass=&NiagaraSpriteRendererProperties, rendererIndex=2}
  ]
 }
 Smoke_Ring002 {
  emitterUpdateScript [
   &EmitterState(
    Inactive Response=::NewEnumerator1
    Life Cycle Mode=::NewEnumerator1
    Loop Behavior=::NewEnumerator1
    Loop Delay=1
    Loop Duration=8
    UseLoopDelay=true
   )
   &SpawnRate(
    SpawnRate=&FloatFromCurve(
     FloatCurve=DataInterface() {propertyValues {Curve {keys [<at 0 1 RCIM_Constant>,<at 0.1 0.4 RCIM_Constant>,<at 0.2 0.2 RCIM_Constant>,<at 0.3 0.95 RCIM_Constant>,<at 0.4 0.1 RCIM_Constant>,<at 0.6 0.75 RCIM_Constant>,<at 0.7 0 RCIM_Constant>,<at 0.8 0.7 RCIM_Constant>,<at 0.9 0 RCIM_Constant>]}}}
     CurveIndex=$Emitter.NormalizedLoopAge
     Scale Curve=12
    )
    Use Spawn Probability=false
   )
  ]
  particleSpawnScript [
   &InitializeParticle(
    Lifetime Max=1
    Lifetime Min=0.6
    Lifetime Mode=::NewEnumerator1
    Mesh Scale=<1 1 5>
    Mesh Scale Mode=::NewEnumerator4
    Position Offset=<0 -2 15>
    UsePositionOffset=true
    Write Material Random=true
   ) {module=InitializeParticle001}
   &InitialMeshOrientation(
    Mesh Orientation Mode=::NewEnumerator2
    Rotation=&RandomRangeVector(Minimum=<0.25 -0.02 -0.02>, Maximum=<0.3 0.02 0.02>)
    Use Rotation Vector=true
   )
   &InitialMeshOrientation(
    Mesh Orientation Mode=::NewEnumerator2
    Rotation=&RandomRangeVector(Minimum=<0 0 0>, Maximum=<0 1 0>)
    Use Rotation Vector=true
   ) {module=InitialMeshOrientation001}
   &CameraOffset(Camera Offset Amount=&FN_FrontEndFloatModifier(Gameplay=0, Front End=5))
  ]
  particleUpdateScript [
   &ParticleState()
   &UpdateMeshOrientation(Rotation Rate=&RandomRangeFloat(Minimum=-0.2, Maximum=0.5), Rotation Vector=<0 0 1>)
   &ScaleMeshSize(Scale Factor=&Multiply_VectorByFloat(
     Vector=&Multiply_VectorByFloat(
      Vector=&MakeVector(
       X=1
       Y=1
       Z=&Multiply_Float(
        A=&FloatFromCurve(
         FloatCurve=DataInterface() {propertyValues {Curve {keys [{arriveTangent=-1.572, time=0, value=1}, {interpMode=RCIM_Cubic, tangentMode=RCTM_User, time=1, value=0.5}]}}}
         Scale Curve=0.7
        )
        B=&RandomRangeFloat(Minimum=1, Maximum=1)
       )
      )
      Float=0.8
     )
     Float=&FloatFromCurve(
      FloatCurve=DataInterface() {propertyValues {Curve {keys [
          {arriveTangent=3.1963, interpMode=RCIM_Cubic, leaveTangent=3.1963, tangentMode=RCTM_User, time=0, value=0.25}
          {arriveTangent=1.8407, interpMode=RCIM_Cubic, leaveTangent=1.8407, tangentMode=RCTM_User, time=1, value=3}
         ]}}}
      Scale Curve=&Multiply_Float(A=&RandomRangeFloat(Minimum=1.2, Maximum=1.5), B=1)
     )
    ))
   &ScaleColor(
    Scale Alpha=&FloatFromCurve(
     FloatCurve=DataInterface() {propertyValues {Curve {keys [
         {arriveTangent=6.8924, interpMode=RCIM_Cubic, leaveTangent=6.8924, tangentMode=RCTM_User, time=0.02, value=0}
         {arriveTangent=1.141, interpMode=RCIM_Cubic, leaveTangent=1.141, tangentMode=RCTM_User, time=0.1, value=0.6}
         {arriveTangent=-1.2399, interpMode=RCIM_Cubic, leaveTangent=-1.2399, tangentMode=RCTM_User, time=0.7, value=0.45}
         {arriveTangent=-1.2361, interpMode=RCIM_Cubic, leaveTangent=-1.2361, tangentMode=RCTM_User, time=1, value=0}
        ]}}}
     Scale Curve=&Multiply_Float(A=&FN_FrontEndFloatModifier(Gameplay=1.8, Front End=2), B=1)
    )
    Scale RGB=&MakeVectorFromLinearColorRGB(Color=&Multiply_LinearColorByFloat(
      Linear Color=&ColorFromCurve(ColorCurve=DataInterface() {propertyValues {
         AlphaCurve {keys []}
         BlueCurve {keys [<at 0.3187 3 RCIM_Cubic>,<at 0.5358 0.3906>,<at 0.8581 0.02 RCIM_Cubic>]}
         GreenCurve {keys [<at 0.3187 0 RCIM_Cubic>,<at 0.5358 0>,<at 0.8581 0 RCIM_Cubic>]}
         RedCurve {keys [<at 0.3187 0.643 RCIM_Cubic>,<at 0.5358 0.0837>,<at 0.8581 0.0043 RCIM_Cubic>]}
        }})
      Float=&FN_FrontEndFloatModifier(Gameplay=1, Front End=0.5)
     ))
    ScaleA=true
    ScaleRGB=true
   )
  ]
  rendererClasses [&NiagaraMeshRendererProperties]
  renderers [{rendererClass=&NiagaraMeshRendererProperties, rendererIndex=0}]
 }
}
summary {
 systemName=NS_Emote_WormChalk_loop
 emitters [
  {emitterName=EdgeGlow, rendererClasses [&NiagaraMeshRendererProperties]}
  {emitterName=Black, rendererClasses [&NiagaraSpriteRendererProperties]}
  {emitterName=Blue, rendererClasses [&NiagaraSpriteRendererProperties]}
  {emitterName=Smoke_Ring002, rendererClasses [&NiagaraMeshRendererProperties]}
  {emitterName=Flame_Explode, rendererClasses [&NiagaraSpriteRendererProperties]}
  {emitterName=BlackFlame002, rendererClasses [&NiagaraSpriteRendererProperties]}
 ]
 userVariables [
  User.bIsFrontEnd {type=NiagaraBool}
  User.bIsFrontEndPreview {type=NiagaraBool}
 ]
}
