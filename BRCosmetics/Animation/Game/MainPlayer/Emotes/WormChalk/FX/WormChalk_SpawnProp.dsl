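; WormChalk_SpawnProp is a Blueprint in the WormChalk emote's FX that customizes the mesh component of a spawned emote prop.
; It overrides CustomizePropMeshComponent, runs the parent behavior, then sets a custom depth stencil value and enables custom depth rendering on the mesh.
; This tags the prop in the custom stencil buffer; the file does not show what reads that tag.
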(fn CustomizePropMeshComponent
  (params (MeshComponent :type Mesh%20Component%20Object%20Reference))
  (|CallParentFunction MeshComponent)
  (Rendering|SetCustomDepthStencilValue MeshComponent 38)
  (Rendering|SetRenderCustomDepth MeshComponent true))
